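# Unlit translucent Niagara material for the WiryPerk skydiving contrail, rendering a trail of grid-quantized square pixels.
# Color blends a two-color gradient times panning noise with optional random three-color "Tetris" blocks, reseeded over particle-relative or global time.
# Opacity combines the pixel grid, depth fade and two erosion steps that threshold gradients against per-cell random values.
# Dynamic parameters drive erosion, reseed rate and optional mask-texture blending, and panning time can be stepped to an FPS.
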Material /BRCosmetics/Athena/Effects/Particles/Skydiving/NS_Contrail_WiryPerk/Materials/M_Contrail_WiryPerk_PixelTrail
graph {
 properties {
  bAutomaticallySetUsageInEditor=false
  bUsedWithNiagaraMeshParticles=true
  bUsedWithNiagaraRibbons=true
  bUsedWithNiagaraSprites=true
  blendMode=BLEND_Translucent
  shadingModel=MSM_Unlit
  usedShadingModels=Unlit
 }
 nodes [
  n1 = TextureSampleParameter2D(UVs=MaterialFunctionCall(UVs=TextureCoordinate(), "Rotation Angle (0-1)"=ScalarParameter() {defaultValue=0.75, group=Mask, parameterName=Mask Rotate}) {materialFunction=&CustomRotator}.Rotated Values) {group=Mask, parameterName=Mask Texture, texture=/BRCosmetics/Animation/Game/MainPlayer/Emotes/WiryPerk/VFX/T_WiryPerk_GG}
  n2 = TextureCoordinate()
  n3 = NamedRerouteUsage()
  n4 = Divide(A=Ceil(None=Multiply(A=@n2, B=@n3)), B=@n3)
  n5 = MaterialFunctionCall(UV Channel=@n4) {materialFunction=&LinearGradient}
  n6 = ScalarParameter() {defaultValue=3, group=Color, parameterName=Color Gradient Power}
  n7 = NamedRerouteUsage()
  n8 = AppendVector(
   A=ScalarParameter() {defaultValue=-1, group=Color, parameterName=Color Speed X}
   B=ScalarParameter() {group=Color, parameterName=Color Speed Y}
  )
  n9 = TextureSampleParameter2D(UVs=Add(A=Panner(Coordinate=@n4, Time=@n7, Speed=@n8), B=NamedRerouteUsage())) {group=Color, parameterName=Color Texture, samplerType=SAMPLERTYPE_Masks, texture=/Game/Global/Textures/Default/Noise/WindNoise_Mask_1}
  n10 = Multiply(
   A=LinearInterpolate(
    A=VectorParameter() {group=Color, parameterName=Base Gradient Color A, defaultValue=<rgba 0 0.5469 0.0131 1>}.RGB
    B=VectorParameter() {group=Color, parameterName=Base Gradient Color B, defaultValue=<rgba 0.1927 1 0.7024 1>}.RGB
    Alpha=Power(Base=@n5.UGradient, Exp=@n6)
   )
   B=@n9.R
  )
  n11 = TextureCoordinate()
  n12 = Multiply(A=NamedRerouteUsage()) {constB=0.25}
  n13 = DynamicParameter() {defaultValue=<rgba 1 0.3 1 25>, paramNames ["Erosion (0-1)", "Erosion Left (0-1)", Param3, PixelPanSpeed]}
  n14 = Multiply(A=ParticleRelativeTime(), B=@n13.PixelPanSpeed) {constB=5}
  n15 = Time()
  MaterialExpressionAppendVector_12 = AppendVector(
   A=Add(A=Multiply(A=@n11, B=@n12) {constB=12}) {constB=0.5}
   B=Round(None=StaticSwitchParameter(True=@n14, False=@n15) {defaultValue=true, group=Time, parameterName=UseParticleRelativeTime})
  )
  n16 = Custom!(v=@MaterialExpressionAppendVector_12) {
   code="// The below expression will get compiled\r\n// into the output of this node\r\nreturn RandFast(uint2(0.5+v.x+v.z*3,0.5+v.y+v.z));"
   showCode=true
   inputs [{
     inputName=v
     input {expression=&self:MaterialExpressionAppendVector_12}
    }]
  }
  n17 = MaterialFunctionCall(
   A=VectorParameter() {group=Color, parameterName=Tetris A Color, defaultValue=<rgba 1 0 0.8569 1>}.RGB
   B=VectorParameter() {group=Color, parameterName=Tetris B Color, defaultValue=<rgba 0.0021 0.1273 1 1>}.RGB
   C=VectorParameter() {group=Color, parameterName=Tetris C Color, defaultValue=<rgba 1 0.8633 0 1>}.RGB
   Alpha=@n16
  ) {materialFunction=&Lerp_3Color}
  n18 = TextureCoordinate()
  n19 = NamedRerouteUsage()
  n20 = Multiply(A=ParticleRelativeTime(), B=@n13.PixelPanSpeed) {constB=5}
  n21 = Time()
  MaterialExpressionAppendVector_11 = AppendVector(
   A=Add(A=Multiply(A=@n18, B=@n19) {constB=12}) {constB=0.5}
   B=Round(None=StaticSwitchParameter(True=@n20, False=@n21) {defaultValue=true, group=Time, parameterName=UseParticleRelativeTime})
  )
  n22 = Custom(v=@MaterialExpressionAppendVector_11) {
   code="// The below expression will get compiled\r\n// into the output of this node\r\nreturn RandFast(uint2(0.5+v.x+v.z*3,0.5+v.y+v.z));"
   showCode=true
   inputs [{
     inputName=v
     input {expression=&self:MaterialExpressionAppendVector_11}
    }]
  }
  n23 = ScalarParameter() {defaultValue=0.8, group=Base Pixels, parameterName=Pixel Size, sortPriority=2}
  n24 = NamedRerouteUsage()
  n25 = Multiply(A=TextureCoordinate(), B=@n24)
  n26 = Frac(None=@n25)
  n27 = MaterialFunctionCall(Float2=Subtract(A=Multiply(A=@n26) {constB=2})) {materialFunction=&BreakOutFloat2Components}
  n28 = SmoothStep(Min=@n23, Max=@n23, Value=@n27.R)
  n29 = SmoothStep(Min=@n23, Max=@n23, Value=@n27.G)
  n30 = Saturate(None=Multiply(A=OneMinus(None=@n28), B=OneMinus(None=@n29)))
  n31 = Divide(B=@n24)
  n32 = Add(A=Multiply(A=Floor(None=@n25), B=@n31), B=@n31)
  n33 = ScalarParameter() {defaultValue=1, group=Base Pixels, parameterName=Pixel Texture Scale}
  n34 = ScalarParameter() {defaultValue=-1, group=Base Pixels, parameterName=Pixel Texture Speed X}
  n35 = ScalarParameter() {group=Base Pixels, parameterName=Pixel Texture Speed Y}
  n36 = TextureSampleParameter2D(UVs=Panner(Coordinate=Multiply(A=@n32, B=@n33), Time=NamedRerouteUsage(), Speed=AppendVector(A=@n34, B=@n35))) {
   group=Base Pixels
   parameterName=Pixel Texture
   samplerType=SAMPLERTYPE_LinearGrayscale
   texture=/Game/Global/Textures/Default/Noise/T_SwirlNoise_H2
  }
  n37 = StaticSwitchParameter(True=Multiply(A=@n30, B=@n36.R), False=@n30) {group=Base Pixels, parameterName=UsePixelTexture}
  n38 = Multiply(A=If(A=@n22, "A > B"=Constant(), "A == B"=Constant() {r=0.5}, "A < B"=Constant() {r=1}) {constB=0.3}, B=@n37)
  n39 = Multiply(
   A=StaticSwitchParameter(True=LinearInterpolate(A=@n10, B=@n17.Result, Alpha=@n38), False=@n10) {defaultValue=true, group=Base Pixels, parameterName=UseTetrisBlocks}
   B=@n37
  )
  n40 = DynamicParameter() {parameterIndex=1, defaultValue=<rgba 1 1 1 25>, paramNames [Param1, LERP_Emissive, LERP_Alpha, Param4]}
  n41 = LinearInterpolate(A=@n1.RGB, B=Multiply(A=@n39, B=@n1.RGB), Alpha=@n40.LERP_Emissive) {constAlpha=1}
  n42 = ParticleColor()
  MaterialFunctionCall(Emissive=Multiply(A=StaticSwitchParameter(True=@n41, False=@n39) {group=Mask, parameterName=MaskTexture_atoggle}, B=@n42.RGB)) {materialFunction=&MF_FX_EmissiveAdjustments}
  n43 = Multiply(A=Time(), B=ScalarParameter() {defaultValue=1, group=Time, parameterName=Time Mult})
  n44 = ScalarParameter() {defaultValue=8, group=Time, parameterName=FPS}
  n45 = Ceil(None=Multiply(A=Frac(None=@n43), B=@n44))
  NamedRerouteDeclaration(None=StaticSwitchParameter(True=@n43, False=Divide(A=@n45, B=@n44)) {group=Time, parameterName=UseNormalTime}) {name=FPS, nodeColor=<rgba 0 0.9531 1 1>}
  n46 = TextureCoordinate()
  n47 = NamedRerouteUsage()
  n48 = Divide(A=Ceil(None=Multiply(A=@n46, B=@n47)), B=@n47)
  n49 = ScalarParameter() {defaultValue=0.3, group=Noise Displacement, parameterName=Noise Scale}
  n50 = ScalarParameter() {defaultValue=-1, group=Noise Displacement, parameterName=Noise Speed X}
  n51 = ScalarParameter() {group=Noise Displacement, parameterName=Noise Speed Y}
  n52 = TextureSampleParameter2D(UVs=Panner(Coordinate=Multiply(A=@n48, B=@n49), Time=NamedRerouteUsage(), Speed=AppendVector(A=@n50, B=@n51))) {
   group=Noise Displacement
   parameterName=Noise Texture
   samplerType=SAMPLERTYPE_LinearColor
   texture=/Game/Global/Textures/Default/Noise/T_FN_FabricTile_B
  }
  NamedRerouteDeclaration(None=Multiply(A=@n52.R, B=ScalarParameter() {group=Noise Displacement, parameterName=Noise Intensity})) {name=Noise, nodeColor=<rgba 0.4062 1 0 1>}
  NamedRerouteDeclaration(None=AppendVector(
    A=ScalarParameter() {defaultValue=16, group=Base Pixels, parameterName=Pixel Tile X, sortPriority=0}
    B=ScalarParameter() {defaultValue=16, group=Base Pixels, parameterName=Pixel Tile Y, sortPriority=1}
   )) {name=Pixel Tile, nodeColor=<rgba 0.1719 0 1 1>}
  n53 = StaticComponentMaskParameter(None=@n1.RGBA) {defaultA=true, group=Mask, parameterName=Channel_atoggle}
  n54 = NamedRerouteUsage()
  n55 = Multiply(A=TextureCoordinate(), B=@n54)
  n56 = Add(A=Divide(A=Ceil(None=@n55), B=@n54)) {constB=-0.1}
  n57 = MaterialFunctionCall(UV Channel=@n56) {materialFunction=&LinearGradient}
  n58 = StaticSwitchParameter(
   True=MaterialFunctionCall(
    UVs=@n56
    Radius=ScalarParameter() {defaultValue=0.5, group=Erosion, parameterName=Sphere Gradient Radius}
    Density=ScalarParameter() {defaultValue=2.33, group=Erosion, parameterName=Sphere Gradient Density}
    Invert Density=StaticBoolParameter() {defaultValue=true, group=Erosion, parameterName="Invert Density (SB)"}
   ) {materialFunction=&RadialGradientExponential}.RadialGradientExponential
   False=@n57.UGradient
  ) {group=Erosion, parameterName=UseRadialErosion}
  n59 = ScalarParameter() {defaultValue=0.6, group=Erosion, parameterName=Erosion Gradient Power, sortPriority=0}
  n60 = TextureCoordinate()
  n61 = NamedRerouteUsage()
  n62 = StaticSwitchParameter(
   True=Multiply(
    A=ParticleRelativeTime()
    B=DynamicParameter() {defaultValue=<rgba 1 0.3 1 25>, paramNames ["Erosion (0-1)", "Erosion Left (0-1)", Param3, PixelPanSpeed]}.PixelPanSpeed
   ) {constB=5}
   False=Time()
  ) {defaultValue=true, group=Time, parameterName=UseParticleRelativeTime}
  MaterialExpressionAppendVector_17 = AppendVector(A=Add(A=Multiply(A=@n60, B=@n61) {constB=12}) {constB=0.5}, B=Add(A=Round(None=@n62)))
  n63 = Custom(v=@MaterialExpressionAppendVector_17) {
   code="// The below expression will get compiled\r\n// into the output of this node\r\nreturn RandFast(uint2(0.5+v.x+v.z*3,0.5+v.y+v.z));"
   showCode=true
   inputs [{
     inputName=v
     input {expression=&self:MaterialExpressionAppendVector_17}
    }]
  }
  n64 = DynamicParameter() {defaultValue=<rgba 1 0.3 1 25>, paramNames ["Erosion (0-1)", "Erosion Left (0-1)", Param3, PixelPanSpeed]}
  n65 = Step(
   Y=Subtract(A=Power(Base=@n58, Exp=@n59), B=@n63)
   X=StaticSwitchParameter(True=@n64."Erosion (0-1)", False=ScalarParameter() {defaultValue=0.5, group=Erosion, parameterName=Erosion Step, sortPriority=2}) {group=Erosion, parameterName=Erosion DynamicParameter}
  ) {constX=0.1}
  n66 = Subtract(
   A=Power(
    Base=OneMinus(None=@n57.UGradient)
    Exp=ScalarParameter() {defaultValue=3, group=Erosion, parameterName=Erosion Gradient Left Power, sortPriority=1}
   )
   B=@n63
  )
  n67 = StaticSwitchParameter(True=@n64."Erosion Left (0-1)", False=ScalarParameter() {defaultValue=2, group=Erosion, parameterName=Erosion Left Step, sortPriority=3}) {group=Erosion, parameterName=Erosion Left DynamicParameter}
  n68 = DepthFade(
   Opacity=Multiply(A=Multiply(A=@n37, B=@n65), B=Step(Y=@n66, X=@n67) {constX=0.1})
   FadeDistance=ScalarParameter() {defaultValue=25, group=Opacity, parameterName=Depth Fade}
  )
  n69 = DynamicParameter() {parameterIndex=1, defaultValue=<rgba 1 1 1 25>, paramNames [Param1, LERP_Emissive, LERP_Alpha, Param4]}
  n70 = LinearInterpolate(A=@n53, B=Multiply(A=@n53, B=@n68), Alpha=@n69.LERP_Alpha) {constA=1, constAlpha=1}
  Saturate(None=Multiply(A=@n42.A, B=StaticSwitchParameter(True=@n70, False=@n68) {group=Mask, parameterName=MaskTexture_atoggle}))
 ]
}
