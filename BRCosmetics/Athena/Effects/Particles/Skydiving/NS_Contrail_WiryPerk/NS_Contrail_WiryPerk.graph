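# WiryPerk skydiving contrail cosmetic: green ribbons, colored pixel and Tetris-style particles, pixel circles and glow sprites at hand and foot sockets.
# Spawn groups or partitions select each socket on the character's skeletal mesh.
# Ribbons are fixed burst particles re-placed every update along an upward line with sine/cosine sway.
# ParaGlideLeanAlpha drives trail length, motion and some spawn rates, with adjustments for front-end preview, platform and time of day.
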NiagaraSystem /BRCosmetics/Athena/Effects/Particles/Skydiving/NS_Contrail_WiryPerk/NS_Contrail_WiryPerk
dependencies {
 usedDataInterfaces [
  &NiagaraDataInterfaceCurve
  &NiagaraDataInterfacePlatformSet
  &NiagaraDataInterfaceArrayColor
  &NiagaraDataInterfaceMeshRendererInfo
  &NiagaraDataInterfaceSpriteRendererInfo
  &NiagaraDataInterfaceColorCurve
  &NiagaraDataInterfaceCamera
  &NiagaraDataInterfaceArrayFloat3
  &NiagaraDataInterfaceSkeletalMesh
 ]
 usedDynamicInputs []
 usedModules [
  &self:SystemSpawnScript.SystemScriptSource.SystemScriptGraph.NiagaraNodeAssignment_0.SetVariables_#1
  &SystemState
  &InterpolateOverTime
  /Game/Effects/Niagara/Modules/Contrails/FN_Contrail_Setup
  &FN_Integrator
 ]
}
emitters {
 Decal {
  emitterUpdateScript [
   &EmitterState(Life Cycle Mode=::NewEnumerator1, Loop Behavior=::NewEnumerator1, Loop Duration Mode=::NewEnumerator1, UseLoopDelay=false)
   &SpawnBurst_Instantaneous(Spawn Count=8, Use Spawn Probability=false, UseLoopCountLimit=false)
  ]
  particleSpawnScript [
   &PartitionParticles(Number Of Partitions=4)
   &InitializeParticle(Color Mode=::NewEnumerator0, Sprite Size Mode=::NewEnumerator3, Uniform Sprite Size=50, UsePositionOffset=false, Write Material Random=true)
   &FN_Sprite_Attributes(
    Camera Offset=-10
    Color=&MakeCustomLinearColorFromBool(Bool=$User.bIsFrontEndPreview, True Color=<rgba 0.1886 1 0.0622 0.1>, False Color=<rgba 0.2572 1 0.0575 0.2>)
    Create Camera Offset Attribute=true
    Create Color Attribute=true
    Create SpriteSize Attribute=true
    SpriteSize=&Vector2DFromFloat(Value=50)
    Write to Module Namespace=true
   )
   &CameraOffset(Camera Offset Amount=20)
  ]
  particleUpdateScript [
   /SetVariables(Particles.LightRadius=&Multiply_Float(A=80, B=&FloatFromCurve(FloatCurve=DataInterface() {propertyValues {Curve {keys [<at 0 0>,<at 1 1>]}}}))) {isSetParameters=true}
   &ParticleState(Kill Particles When Lifetime Has Elapsed=false, Let Infinitely Lived Particles Die When Emitter Deactivates=true)
   &SkeletalMeshLocation(
    Bone / Socket Index=$Particles.PartitionParticles.Partition
    MaskSpawnGroup=false
    Mesh Sampling Type=::NewEnumerator5
    Sampled Position Offset=&SelectVectorFromArray(
     Direct Array Index=$Particles.PartitionParticles.Partition
     Vector Selection Array=DataInterface() {propertyValues {FloatData [<0 10 5>, <0 10 5>, <0 0 5>, <0 0 5>]}}
    )
    Sampled Position Offset Coordinate Space:ENiagaraSkelMeshTransforms=::NewEnumerator2
    Skeletal Mesh=$System.Skeletal Mesh
    Socket Sampling Mode=::NewEnumerator5
   )
   &ScaleColor(
    Scale Alpha=&Multiply_Float(
     A=&FloatFromCurve(FloatCurve=DataInterface() {propertyValues {Curve {keys [<at 0 0>,<at 0.2317 1>,<at 1 1>]}}})
     B=1
    )
    Scale RGB=&MakeVectorFromLinearColorRGB(Color=<rgba 0.7183 2 0.0895 1>)
    ScaleA=true
    ScaleRGB=true
   )
  ]
  rendererClasses [&NiagaraSpriteRendererProperties]
  renderers [
   {rendererClass=&NiagaraSpriteRendererProperties}
   {rendererClass=&NiagaraSpriteRendererProperties, rendererIndex=1}
  ]
 }
 Pixel_Circles {
  emitterUpdateScript [
   &EmitterState(Life Cycle Mode=::NewEnumerator1, Loop Behavior=::NewEnumerator1, Loop Duration Mode=::NewEnumerator1, UseLoopDelay=false)
   &SpawnRate(SpawnRate=&Lerp_Float(A=1.4, B=3, Alpha=$User.ParaGlideLeanAlpha), Use Spawn Probability=false)
   &SpawnRate(SpawnGroup=1, SpawnRate=&Lerp_Float(A=1.4, B=3, Alpha=$User.ParaGlideLeanAlpha), Use Spawn Probability=false) {module=SpawnRate001}
   &SpawnRate(SpawnGroup=2, SpawnRate=&Lerp_Float(A=1.4, B=3, Alpha=$User.ParaGlideLeanAlpha), Use Spawn Probability=false) {module=SpawnRate002}
   &SpawnRate(SpawnGroup=3, SpawnRate=&Lerp_Float(A=1.4, B=3, Alpha=$User.ParaGlideLeanAlpha), Use Spawn Probability=false) {module=SpawnRate003}
   &SpawnBurst_Instantaneous(Use Spawn Probability=false, UseLoopCountLimit=false) {enabled=false}
   &SpawnBurst_Instantaneous(Spawn Group=1, Use Spawn Probability=false, UseLoopCountLimit=false) {enabled=false, module=SpawnBurst_Instantaneous001}
   &SpawnBurst_Instantaneous(Spawn Group=2, Use Spawn Probability=false, UseLoopCountLimit=false) {enabled=false, module=SpawnBurst_Instantaneous002}
   &SpawnBurst_Instantaneous(Spawn Group=3, Use Spawn Probability=false, UseLoopCountLimit=false) {enabled=false, module=SpawnBurst_Instantaneous003}
  ]
  particleSpawnScript [&InitializeParticle(
    Color=<rgba 2.5 2.5 2.5 1>
    Lifetime=&Lerp_Float(A=0.85, B=0.525, Alpha=$User.ParaGlideLeanAlpha)
    Mesh Scale Mode=::NewEnumerator3
    Mesh Uniform Scale=/Game/Effects/Niagara/DynamicInputs/FN_PickFloatFromSpawnGroupCutoff(
     Over Cutoff Float:NiagaraFloat=0.175
     Under Cutoff Float:NiagaraFloat=0.275
     SpawnGroup:NiagaraInt32=$Emitter.SpawnGroup
     Cutoff:NiagaraInt32=2
    )
    UsePositionOffset=false
    Write Material Random=true
   )]
  particleUpdateScript [
   &ParticleState()
   &ScaleMeshSize(Scale Factor=&VectorFromFloat(Value=&FloatFromCurve(FloatCurve=DataInterface() {propertyValues {Curve {keys [<at 0 0.4 RCIM_Cubic>,<at 1 1 RCIM_Cubic>]}}}, Scale Curve=1.35)))
   &SkeletalMeshLocation(
    Bone / Socket Index=&SpawnGroup()
    MaskSpawnGroup=false
    Mesh Sampling Type=::NewEnumerator5
    Orientation Sampling=::NewEnumerator1
    Sampled Position Offset=/Game/Effects/Niagara/DynamicInputs/Water/FN_PickVectorFromSpawnGroupCutoff(
     Under Cutoff Vector:Vector3f=&Lerp_Vector(
      A=<0 5 0>
      B=&FN_FrontEndVector3Modifier(Gameplay=<0 0 15>, Front End=<0 0 0>)
      Alpha=$User.ParaGlideLeanAlpha
     )
     Over Cutoff Vector:Vector3f=&Lerp_Vector(
      A=<0 -5 10>
      B=&FN_FrontEndVector3Modifier(Gameplay=<0 6 15>, Front End=<0 -6 10>)
      Alpha=$User.ParaGlideLeanAlpha
     )
     SpawnGroup:NiagaraInt32=$Emitter.SpawnGroup
     Cutoff:NiagaraInt32=2
    )
    Sampled Position Offset Coordinate Space:ENiagaraSkelMeshTransforms=::NewEnumerator2
    Skeletal Mesh=$System.Skeletal Mesh
    Socket Sampling Mode=::NewEnumerator5
   )
   &DynamicMaterialParameters(
    Index 0 Param 4=&Lerp_Float(A=15, B=18, Alpha=$User.ParaGlideLeanAlpha)
    Module.Index0Float0Write=true
    Module.Index0Float1Write=true
    Module.Index0Float2Write=true
    Module.Index0Float3Write=true
   )
   &ScaleColor(
    Scale Alpha=&FloatFromCurve(FloatCurve=DataInterface() {propertyValues {Curve {
        defaultValue=1
        keys [
         {arriveTangent=9.0725, interpMode=RCIM_Cubic, leaveTangent=9.0725, tangentMode=RCTM_User, time=0, value=0}
         {arriveTangent=0.0014, interpMode=RCIM_Cubic, leaveTangent=0.0014, tangentMode=RCTM_User, time=0.1408, value=0.9847}
         {arriveTangent=0.1979, interpMode=RCIM_Cubic, leaveTangent=0.1979, tangentMode=RCTM_User, time=0.75, value=1}
         {arriveTangent=-3.9187, interpMode=RCIM_Cubic, leaveTangent=-3.9187, tangentMode=RCTM_User, time=1, value=0}
        ]
       }}})
    ScaleA=true
    ScaleRGB=true
   )
  ]
  rendererClasses [&NiagaraMeshRendererProperties]
  renderers [{rendererClass=&NiagaraMeshRendererProperties, rendererIndex=0}]
 }
 Pixels_HandsFeet {
  emitterUpdateScript [
   &EmitterState(Life Cycle Mode=::NewEnumerator1, Loop Behavior=::NewEnumerator1, Loop Duration Mode=::NewEnumerator1, UseLoopDelay=false)
   &SpawnRate(SpawnRate=128, Use Spawn Probability=false)
   &SpawnRate(SpawnGroup=1, SpawnRate=128, Use Spawn Probability=false) {module=SpawnRate001}
   &SpawnRate(SpawnGroup=2, SpawnRate=48, Use Spawn Probability=false) {module=SpawnRate002}
   &SpawnRate(SpawnGroup=3, SpawnRate=48, Use Spawn Probability=false) {module=SpawnRate003}
  ]
  particleSpawnScript [
   &InitializeParticle(
    Color=&SelectLinearColorFromArray(
     Array Sampling Mode=::NewEnumerator0
     Direct Array Index=?
     Color Selection Array=DataInterface() {propertyValues {ColorData [<rgba 1 0.9141 0.25 1>, <rgba 0.1631 1 0.11 1>]}}
     Random Seed=0
    )
    Lifetime Max=0.35
    Lifetime Min=0.1
    Lifetime Mode=::NewEnumerator1
    Mesh Scale Mode=::NewEnumerator1
    Mesh Uniform Scale Max=0.125
    Mesh Uniform Scale Min=0.05
    Sprite Rotation Mode=::NewEnumerator2
    Sprite Size Mode=::NewEnumerator1
    Uniform Sprite Size Max=9
    Uniform Sprite Size Min=6
    UsePositionOffset=false
    Write Material Random=true
   )
   &ShapeLocation(MaskSpawnGroup=false, Override Seed=false, Sphere Radius=8, UseOwnerScale=true)
   &AddVelocity(
    Module.Use Falloff Distance=false
    Origin Offset:NiagaraPosition=$Particles.SkeletalMeshLocation.SampledPosition
    Override Seed=false
    Velocity Mode=::NewEnumerator1
    Velocity Speed=&FN_PickFloatFromSpawnGroupCutoff(
     Over Cutoff Float:NiagaraFloat=&RandomRangeFloat(Minimum=25, Maximum=75)
     Under Cutoff Float:NiagaraFloat=&RandomRangeFloat(Minimum=50, Maximum=90)
     SpawnGroup:NiagaraInt32=$Emitter.SpawnGroup
     Cutoff:NiagaraInt32=2
    )
   ) {module=AddVelocity001}
   &InitialMeshOrientation(Use Rotation Vector=false)
  ]
  particleUpdateScript [
   &ParticleState()
   &SkeletalMeshLocation(
    Bone / Socket Index=&SpawnGroup()
    MaskSpawnGroup=false
    Mesh Sampling Type=::NewEnumerator5
    Position Sampling=::NewEnumerator1
    Sampled Position Offset=&AddVectorToPosition(
     Position=$Particles.ShapeLocation.ShapePosition
     Vector=&FN_PickVectorFromSpawnGroupCutoff(
      Under Cutoff Vector:Vector3f=<0 0 0>
      Over Cutoff Vector:Vector3f=<0 0 15>
      SpawnGroup:NiagaraInt32=$Emitter.SpawnGroup
      Cutoff:NiagaraInt32=2
     )
    )
    Sampled Position Offset Coordinate Space:ENiagaraSkelMeshTransforms=::NewEnumerator2
    Skeletal Mesh=$System.Skeletal Mesh
    Socket Sampling Mode=::NewEnumerator5
   )
   &ScaleMeshSize(Scale Factor=&VectorFromFloat(Value=&FloatFromCurve(
      FloatCurve=DataInterface() {propertyValues {Curve {
         defaultValue=0
         keys [
          {tangentMode=RCTM_User, time=0, value=1}
          {arriveTangent=-0.0405, interpMode=RCIM_Cubic, leaveTangent=-0.0405, tangentMode=RCTM_User, time=0.7009, value=1}
          {arriveTangent=-8.0089, interpMode=RCIM_Cubic, leaveTangent=-8.0089, tangentMode=RCTM_User, time=1, value=0}
         ]
        }}}
      Scale Curve=&Lerp_Float(A=1, B=0.7, Alpha=$User.ParaGlideLeanAlpha)
     )))
   &MeshRotationForce(Rotation=&VectorFromFloat(Value=&RandomRangeFloat(Minimum=-10, Maximum=10)))
   &CurlNoiseForce(
    Bias Noise=false
    Noise Strength=&FloatFromCurve(
     FloatCurve=DataInterface() {propertyValues {Curve {keys [<at 0 0>,<at 0.2 1 RCIM_Cubic>]}}}
     Scale Curve=&RandomRangeFloat(Minimum=125, Maximum=300)
    )
    Pan Noise=false
   )
   &GravityForce(Gravity=&MakeVector(
     X=0
     Y=0
     Z=&FloatFromCurve(
      FloatCurve=DataInterface() {propertyValues {Curve {keys [<at 0 0 RCIM_Cubic>,<at 0.2 1 RCIM_Cubic>]}}}
      Scale Curve=&Lerp_Float(A=980, B=1150, Alpha=$User.ParaGlideLeanAlpha)
     )
    ))
   &Drag(Drag=&Lerp_Float(A=2, B=0.8, Alpha=$User.ParaGlideLeanAlpha), Use Linear Drag=true, Use Rotational Drag=true)
   &SolveForcesAndVelocity(Clamp Velocity=false, Limit Acceleration=false)
   &ScaleColor(
    Linear Color Curve=DataInterface() {propertyValues {
      AlphaCurve {keys [<at 0 0>,<at 0.2 1>,<at 1 0>]}
      BlueCurve {keys [<at 0 1>,<at 1 1>]}
      GreenCurve {keys [<at 0 1>,<at 1 1>]}
      RedCurve {keys [<at 0 1>,<at 1 1>]}
     }}
    Scale Mode=::NewEnumerator2
   ) {enabled=false}
   &Color(Color=&MultiplyLinearColor(
     Linear Color=&SelectLinearColorFromArray(
      Array Sampling Mode=::NewEnumerator0
      Direct Array Index=?
      Color Selection Array=DataInterface() {propertyValues {ColorData [<rgba 0.3226 0.6488 1 1>, <rgba 1 0.9141 0.25 1>, <rgba 1 0.1771 0.7738 1>, <rgba 0.4684 1 0.2654 1>]}}
      Random Seed=0
     )
     Scale Factor=&Multiply_Vector4(
      A=&MakeVector4(
       X=1
       Y=1
       Z=1
       W=&FloatFromCurve(
        FloatCurve=DataInterface() {propertyValues {Curve {keys [<at 0 0 RCIM_Cubic>,<at 0.2 1 RCIM_Cubic>,<at 1 0 RCIM_Cubic>]}}}
        Scale Curve=/Niagara/DynamicInputs/Camera/CameraFade(
         Camera Query:NiagaraDataInterfaceCamera=DataInterface()
         Fade Offset:NiagaraFloat=50
         Fade Distance:NiagaraFloat=125
         Use Owner Position:NiagaraBool=true
        )
       )
      )
      B=&MakeVector4_FromFloat(Float=&FN_Select_Float_from_Time_Of_Day(Morning=1.7, Day=1, Evening=1.7, Night=2.5))
     )
    ))
  ]
  rendererClasses [&NiagaraMeshRendererProperties]
  renderers [{rendererClass=&NiagaraMeshRendererProperties, rendererIndex=0}]
 }
 Ribbon_Feet {
  emitterUpdateScript [
   &EmitterState(
    Inactive Response=::NewEnumerator1
    Life Cycle Mode=::NewEnumerator1
    Loop Behavior=::NewEnumerator1
    Loop Duration Mode=::NewEnumerator1
    UseLoopDelay=false
   )
   &SpawnBurst_Instantaneous(Spawn Count=5, Spawn Group=2, Use Spawn Probability=false, UseLoopCountLimit=false)
   &SpawnBurst_Instantaneous(Spawn Count=5, Spawn Group=3, Use Spawn Probability=false, UseLoopCountLimit=false) {module=SpawnBurst_Instantaneous001}
  ]
  particleSpawnScript [
   /SetVariables(Particles.RibbonLinkOrder:NiagaraFloat=&ReturnNormalizedExecIndex()) {isSetParameters=true, module=SetVariables_#2}
   &InitializeParticle(
    Ribbon Width=&FloatFromCurve(
     FloatCurve=DataInterface() {propertyValues {Curve {keys [
         {arriveTangent=1.2674, interpMode=RCIM_Cubic, leaveTangent=1.2674, tangentMode=RCTM_User, time=0, value=0.45}
         {arriveTangent=0.25, interpMode=RCIM_Cubic, leaveTangent=0.25, time=0.8, value=1.5}
         {arriveTangent=-7.8136, interpMode=RCIM_Cubic, leaveTangent=-7.8136, tangentMode=RCTM_User, time=1, value=0.7}
        ]}}}
     CurveIndex=$Particles.RibbonLinkOrder
     Scale Curve=&MakeCustomFloatFromBool(Bool=&SetBoolByFloatComparison(A=1, B=$User.ParaGlideLeanAlpha), True Float=9, False Float=12)
    )
    Ribbon Width Mode=::NewEnumerator1
    UsePositionOffset=false
    Write Material Random=true
   )
   /SetVariables(Particles.RibbonWidth002:NiagaraFloat=&FloatFromCurve(
     FloatCurve=DataInterface() {propertyValues {Curve {keys [
         {arriveTangent=1.2674, interpMode=RCIM_Cubic, leaveTangent=1.2674, tangentMode=RCTM_User, time=0, value=0.45}
         {arriveTangent=0.25, interpMode=RCIM_Cubic, leaveTangent=0.25, time=0.8, value=1.5}
         {arriveTangent=-7.8136, interpMode=RCIM_Cubic, leaveTangent=-7.8136, tangentMode=RCTM_User, time=1, value=0.7}
        ]}}}
     CurveIndex=$StackContext.RibbonLinkOrder
     Scale Curve=&MakeCustomFloatFromBool(Bool=&SetBoolByFloatComparison(A=1, B=$User.ParaGlideLeanAlpha), True Float=12, False Float=16)
    )) {isSetParameters=true}
  ]
  particleUpdateScript [
   &ParticleState(Kill Particles When Lifetime Has Elapsed=false, Let Infinitely Lived Particles Die When Emitter Deactivates=true)
   &SkeletalMeshLocation(
    Bone / Socket Index=&SpawnGroup()
    "Interpolate Mesh Sampling (Update)"=true
    MaskSpawnGroup=false
    Mesh Sampling Type=::NewEnumerator5
    Position Sampling=::NewEnumerator2
    Sampled Position Offset=&Lerp_Vector(
     A=<0 -8 8>
     B=&FN_FrontEndVector3Modifier(Gameplay=<0 6 5>, Front End=<0 -6 0>)
     Alpha=$User.ParaGlideLeanAlpha
    )
    Sampled Position Offset Coordinate Space:ENiagaraSkelMeshTransforms=::NewEnumerator2
    Skeletal Mesh=$System.Skeletal Mesh
    Socket Sampling Mode=::NewEnumerator5
   )
   /SetVariables(Particles.Position=/Niagara/DynamicInputs/Lerp/Lerp_Position(
     A:NiagaraPosition=$Particles.SkeletalMeshLocation.SampledPosition
     B:NiagaraPosition=&Add_Vector(
      A=Linked() {linkedVariable {name=Particles.SkeletalMeshLocation.SampledPosition, type=NiagaraPosition}}
      B=&Lerp_Vector(
       A=&MakeCustomVectorFromBool(Bool=$User.bIsFrontEndPreview, True Vector=<0 0 50>, False Vector=<0 0 80>)
       B=&MakeCustomVectorFromBool(Bool=$User.bIsFrontEndPreview, True Vector=<0 0 80>, False Vector=<0 0 120>)
       Alpha=$User.ParaGlideLeanAlpha
      )
     )
     Alpha:NiagaraFloat=$Particles.RibbonLinkOrder
    )) {isSetParameters=true, module=SetVariables_#3}
   /SetVariables(Particles.Position=&Add_Vector(
     A=Linked() {linkedVariable {name=Particles.Position, type=NiagaraPosition}}
     B=&MakeVector(
      X=&Sine(
       Normalized Angle=&Add_Float(
        A=&Add_Float(A=$Particles.RibbonLinkOrder, B=&Multiply_Float(A=$Emitter.Age, B=-0.4))
        B=&Multiply_FloatByInt(Float=0.33, Integer=&SpawnGroup())
       )
       Period=2
       Scale=&Multiply_Float(A=$Particles.RibbonLinkOrder, B=4)
       Bias=0
      )
      Y=&Cosine(
       Normalized Angle=&Add_Float(
        A=&Add_Float(A=$Particles.RibbonLinkOrder, B=&Multiply_Float(A=$Emitter.Age, B=-1))
        B=&Multiply_FloatByInt(Float=0.33, Integer=&SpawnGroup())
       )
       Period=1.3
       Scale=&Multiply_Float(A=$Particles.RibbonLinkOrder, B=15)
       Bias=0
      )
      Z=0
     )
    )) {isSetParameters=true, module=SetVariables_#4}
   &ScaleColor(
    Scale Alpha=&FloatFromCurve(
     FloatCurve=DataInterface() {propertyValues {Curve {keys [
         {arriveTangent=-0.6517, interpMode=RCIM_Cubic, leaveTangent=-0.6517, tangentMode=RCTM_User, time=0, value=1}
         {arriveTangent=-0.7628, interpMode=RCIM_Cubic, leaveTangent=-0.7628, tangentMode=RCTM_User, time=1, value=0}
        ]}}}
     CurveIndex=$Particles.RibbonLinkOrder
    )
    ScaleA=true
    ScaleRGB=true
   )
   /SetVariables(
    Particles.Color003:LinearColor=&MakeLinearColorFromVectorAndFloat(
     "Vector (RGB)"=&MakeVectorFromLinearColorRGB(Color=&FNTimeOfDayColor(
       Night Color=<rgba 0.3358 1 0.1927 1>
       Morning Color=<rgba 0.1773 1 0 1>
       Day Color=<rgba 0.1773 1 0 1>
       Evening Color=<rgba 0.3358 1 0.1927 1>
      ))
     "Float (Alpha)"=&FloatFromCurve(
      FloatCurve=DataInterface() {propertyValues {Curve {keys [
          {arriveTangent=-2.0494, interpMode=RCIM_Cubic, leaveTangent=-2.0494, tangentMode=RCTM_User, time=0, value=1}
          {interpMode=RCIM_Cubic, leaveTangent=-2.5239, time=1, value=0}
         ]}}}
      CurveIndex=$Particles.RibbonLinkOrder
      Scale Curve=&MakeCustomFloatFromBool(
       Bool=$User.bIsFrontEndPreview
       True Float=&FN_PickFloatDependOnPlatform(TargetPlatformActive=0.3, TargetPlatformNoActive=0.2)
       False Float=&FN_PickFloatDependOnPlatform(TargetPlatformActive=0.8, TargetPlatformNoActive=0.6)
      )
     )
    )
    Particles.GlowRibbonDP:Vector4f=<0 0 0 0>
   ) {isSetParameters=true}
   &DynamicMaterialParameters(
    Index 0 Param 4=&Lerp_Float(A=8, B=24, Alpha=$User.ParaGlideLeanAlpha)
    Module.Index0Float0Write=false
    Module.Index0Float1Write=false
    Module.Index0Float2Write=false
    Module.Index0Float3Write=true
   )
  ]
  rendererClasses [&NiagaraRibbonRendererProperties]
  renderers [
   {rendererClass=&NiagaraRibbonRendererProperties}
   {rendererClass=&NiagaraRibbonRendererProperties, rendererIndex=1}
  ]
 }
 Ribbon_Hands {
  emitterUpdateScript [
   &EmitterState(
    Inactive Response=::NewEnumerator1
    Life Cycle Mode=::NewEnumerator1
    Loop Behavior=::NewEnumerator1
    Loop Duration Mode=::NewEnumerator1
    UseLoopDelay=false
   )
   &SpawnBurst_Instantaneous(Spawn Count=5, Use Spawn Probability=false, UseLoopCountLimit=false)
   &SpawnBurst_Instantaneous(Spawn Count=5, Spawn Group=1, Use Spawn Probability=false, UseLoopCountLimit=false) {module=SpawnBurst_Instantaneous001}
  ]
  particleSpawnScript [
   /SetVariables(Particles.RibbonLinkOrder:NiagaraFloat=&ReturnNormalizedExecIndex()) {isSetParameters=true, module=SetVariables_#2}
   &InitializeParticle(
    Ribbon Width=&FloatFromCurve(
     FloatCurve=DataInterface() {propertyValues {Curve {keys [
         {arriveTangent=1.2674, interpMode=RCIM_Cubic, leaveTangent=1.2674, tangentMode=RCTM_User, time=0, value=0.45}
         {arriveTangent=0.25, interpMode=RCIM_Cubic, leaveTangent=0.25, time=0.8, value=1.5}
         {arriveTangent=-7.8136, interpMode=RCIM_Cubic, leaveTangent=-7.8136, tangentMode=RCTM_User, time=1, value=0.7}
        ]}}}
     CurveIndex=$Particles.RibbonLinkOrder
     Scale Curve=&MakeCustomFloatFromBool(Bool=&SetBoolByFloatComparison(A=1, B=$User.ParaGlideLeanAlpha), True Float=16, False Float=18)
    )
    Ribbon Width Mode=::NewEnumerator1
    UsePositionOffset=false
    Write Material Random=true
   )
   /SetVariables(Particles.RibbonWidth002:NiagaraFloat=&FloatFromCurve(
     FloatCurve=DataInterface() {propertyValues {Curve {keys [
         {arriveTangent=1.2674, interpMode=RCIM_Cubic, leaveTangent=1.2674, tangentMode=RCTM_User, time=0, value=0.45}
         {arriveTangent=0.25, interpMode=RCIM_Cubic, leaveTangent=0.25, time=0.8, value=1.5}
         {arriveTangent=-7.8136, interpMode=RCIM_Cubic, leaveTangent=-7.8136, tangentMode=RCTM_User, time=1, value=0.7}
        ]}}}
     CurveIndex=$StackContext.RibbonLinkOrder
     Scale Curve=&MakeCustomFloatFromBool(Bool=&SetBoolByFloatComparison(A=1, B=$User.ParaGlideLeanAlpha), True Float=22, False Float=25)
    )) {isSetParameters=true}
  ]
  particleUpdateScript [
   &ParticleState(Kill Particles When Lifetime Has Elapsed=false, Let Infinitely Lived Particles Die When Emitter Deactivates=true)
   &SkeletalMeshLocation(
    Bone / Socket Index=&SpawnGroup()
    "Interpolate Mesh Sampling (Update)"=true
    MaskSpawnGroup=false
    Mesh Sampling Type=::NewEnumerator5
    Position Sampling=::NewEnumerator2
    Sampled Position Offset=&Lerp_Vector(
     A=&MaskVectorBySpawnGroup(
      Masked Vector=<0 5 5>
      Passthrough Vector=&MaskVectorBySpawnGroup(Masked Vector=<0 5 5>, Passthrough Vector=<0 0 0>, Spawn Group Mask=1)
     )
     B=<0 0 0>
     Alpha=$User.ParaGlideLeanAlpha
    )
    Sampled Position Offset Coordinate Space:ENiagaraSkelMeshTransforms=::NewEnumerator2
    Skeletal Mesh=$System.Skeletal Mesh
    Socket Sampling Mode=::NewEnumerator5
   )
   /SetVariables(Particles.Position=&Lerp_Position(
     A:NiagaraPosition=$Particles.SkeletalMeshLocation.SampledPosition
     B:NiagaraPosition=&Add_Vector(
      A=Linked() {linkedVariable {name=Particles.SkeletalMeshLocation.SampledPosition, type=NiagaraPosition}}
      B=&Lerp_Vector(
       A=&MakeCustomVectorFromBool(Bool=$User.bIsFrontEndPreview, True Vector=<0 0 150>, False Vector=<0 0 180>)
       B=&MakeCustomVectorFromBool(Bool=$User.bIsFrontEndPreview, True Vector=<0 0 200>, False Vector=<0 0 275>)
       Alpha=$User.ParaGlideLeanAlpha
      )
     )
     Alpha:NiagaraFloat=$Particles.RibbonLinkOrder
    )) {isSetParameters=true, module=SetVariables_#3}
   /SetVariables(Particles.Position=&Add_Vector(
     A=Linked() {linkedVariable {name=Particles.Position, type=NiagaraPosition}}
     B=&Lerp_Vector(
      A=&MakeVector(
       X=&Sine(
        Normalized Angle=&Add_Float(
         A=&Add_Float(A=$Particles.RibbonLinkOrder, B=&Multiply_Float(A=$Emitter.Age, B=-1.2))
         B=&Multiply_FloatByInt(Float=0.33, Integer=&SpawnGroup())
        )
        Period=2
        Scale=&Multiply_Float(A=$Particles.RibbonLinkOrder, B=30)
        Bias=0
       )
       Y=&Cosine(
        Normalized Angle=&Add_Float(
         A=&Add_Float(A=$Particles.RibbonLinkOrder, B=&Multiply_Float(A=$Emitter.Age, B=-1))
         B=&Multiply_FloatByInt(Float=0.33, Integer=&SpawnGroup())
        )
        Period=1.3
        Scale=&Multiply_Float(A=$Particles.RibbonLinkOrder, B=22)
        Bias=0
       )
       Z=0
      )
      B=&MakeVector(
       X=&Sine(
        Normalized Angle=&Add_Float(
         A=&Add_Float(A=$Particles.RibbonLinkOrder, B=&Multiply_Float(A=$Emitter.Age, B=-1.2))
         B=&Multiply_FloatByInt(Float=0.33, Integer=&SpawnGroup())
        )
        Period=2
        Scale=&Multiply_Float(A=$Particles.RibbonLinkOrder, B=30)
        Bias=0
       )
       Y=&Cosine(
        Normalized Angle=&Add_Float(
         A=&Add_Float(A=$Particles.RibbonLinkOrder, B=&Multiply_Float(A=$Emitter.Age, B=-1))
         B=&Multiply_FloatByInt(Float=0.33, Integer=&SpawnGroup())
        )
        Period=1.3
        Scale=&Multiply_Float(A=$Particles.RibbonLinkOrder, B=45)
        Bias=0
       )
       Z=0
      )
      Alpha=$User.ParaGlideLeanAlpha
     )
    )) {isSetParameters=true, module=SetVariables_#4}
   &ScaleColor(
    Scale Alpha=&FloatFromCurve(
     FloatCurve=DataInterface() {propertyValues {Curve {keys [
         {arriveTangent=-0.6517, interpMode=RCIM_Cubic, leaveTangent=-0.6517, tangentMode=RCTM_User, time=0, value=1}
         {arriveTangent=-0.7628, interpMode=RCIM_Cubic, leaveTangent=-0.7628, tangentMode=RCTM_User, time=1, value=0}
        ]}}}
     CurveIndex=$Particles.RibbonLinkOrder
     Scale Curve=0.5
    )
    ScaleA=true
    ScaleRGB=true
   )
   /SetVariables(
    Particles.Color003:LinearColor=&MakeLinearColorFromVectorAndFloat(
     "Vector (RGB)"=&MakeVectorFromLinearColorRGB(Color=&FNTimeOfDayColor(
       Night Color=<rgba 0.3358 1 0.1927 1>
       Morning Color=<rgba 0.1773 1 0 1>
       Day Color=<rgba 0.1773 1 0 1>
       Evening Color=<rgba 0.3358 1 0.1927 1>
      ))
     "Float (Alpha)"=&FloatFromCurve(
      FloatCurve=DataInterface() {propertyValues {Curve {keys [
          {arriveTangent=-2.0494, interpMode=RCIM_Cubic, leaveTangent=-2.0494, tangentMode=RCTM_User, time=0, value=1}
          {interpMode=RCIM_Cubic, leaveTangent=-2.5239, time=1, value=0}
         ]}}}
      CurveIndex=$Particles.RibbonLinkOrder
      Scale Curve=&MakeCustomFloatFromBool(
       Bool=$User.bIsFrontEndPreview
       True Float=&FN_PickFloatDependOnPlatform(TargetPlatformActive=0.3, TargetPlatformNoActive=0.2)
       False Float=&FN_PickFloatDependOnPlatform(TargetPlatformActive=0.8, TargetPlatformNoActive=0.6)
      )
     )
    )
    Particles.GlowRibbonDP:Vector4f=<0 0 0 0>
   ) {isSetParameters=true}
   &DynamicMaterialParameters(
    Index 0 Param 4=&Lerp_Float(A=12, B=36, Alpha=$User.ParaGlideLeanAlpha)
    Module.Index0Float0Write=false
    Module.Index0Float1Write=false
    Module.Index0Float2Write=false
    Module.Index0Float3Write=true
   )
  ]
  rendererClasses [&NiagaraRibbonRendererProperties]
  renderers [
   {rendererClass=&NiagaraRibbonRendererProperties}
   {rendererClass=&NiagaraRibbonRendererProperties, rendererIndex=1}
  ]
 }
 Tetris_HandsFeet {
  emitterUpdateScript [
   &EmitterState(Life Cycle Mode=::NewEnumerator1, Loop Behavior=::NewEnumerator1, Loop Duration Mode=::NewEnumerator1, UseLoopDelay=false)
   &SpawnRate(SpawnRate=&Lerp_Float(A=42, B=64, Alpha=$User.ParaGlideLeanAlpha), Use Spawn Probability=false)
   &SpawnRate(SpawnGroup=1, SpawnRate=&Lerp_Float(A=42, B=64, Alpha=$User.ParaGlideLeanAlpha), Use Spawn Probability=false) {module=SpawnRate001}
   &SpawnRate(SpawnGroup=2, SpawnRate=&Lerp_Float(A=6, B=9, Alpha=$User.ParaGlideLeanAlpha), Use Spawn Probability=false) {module=SpawnRate002}
   &SpawnRate(SpawnGroup=3, SpawnRate=&Lerp_Float(A=6, B=9, Alpha=$User.ParaGlideLeanAlpha), Use Spawn Probability=false) {module=SpawnRate003}
  ]
  particleSpawnScript [
   &InitializeParticle(
    Color=&SelectLinearColorFromArray(
     Array Sampling Mode=::NewEnumerator0
     Direct Array Index=?
     Color Selection Array=DataInterface() {propertyValues {ColorData [<rgba 0.3226 0.6488 1 1>, <rgba 1 0.9141 0.25 1>, <rgba 1 0.2708 0.7996 1>]}}
     Random Seed=0
    )
    Lifetime Max=0.6
    Lifetime Min=0.2
    Lifetime Mode=::NewEnumerator1
    Mesh Renderer Array Visibility Mode=::NewEnumerator3
    Mesh Renderer Info=DataInterface() {propertyValues {MeshRenderer=&self:Tetris_HandsFeet_0.NiagaraMeshRendererProperties_0}}
    Mesh Scale Mode=::NewEnumerator1
    Mesh Uniform Scale Max=0.15
    Mesh Uniform Scale Min=0.075
    Sprite Rotation Mode=::NewEnumerator2
    Sprite Size Mode=::NewEnumerator1
    Uniform Sprite Size Max=9
    Uniform Sprite Size Min=6
    UsePositionOffset=false
    Write Material Random=true
   )
   &ShapeLocation(
    MaskSpawnGroup=false
    Non Uniform Scale=<0.65 1 8>
    Override Seed=false
    Sphere Radius=&FN_PickFloatFromSpawnGroupCutoff(Over Cutoff Float:NiagaraFloat=3, Under Cutoff Float:NiagaraFloat=8, SpawnGroup:NiagaraInt32=$Emitter.SpawnGroup, Cutoff:NiagaraInt32=2)
    UseOwnerScale=true
   )
   &SkeletalMeshLocation(
    Bone / Socket Index=&SpawnGroup()
    MaskSpawnGroup=false
    Mesh Sampling Type=::NewEnumerator5
    Sampled Position Offset=&AddVectorToPosition(
     Position=$Particles.ShapeLocation.ShapePosition
     Vector=&MakeVector(
      X=0
      Y=0
      Z=&FN_PickFloatFromSpawnGroupCutoff(Over Cutoff Float:NiagaraFloat=30, Under Cutoff Float:NiagaraFloat=125, SpawnGroup:NiagaraInt32=$Emitter.SpawnGroup, Cutoff:NiagaraInt32=2)
     )
    )
    Sampled Position Offset Coordinate Space:ENiagaraSkelMeshTransforms=::NewEnumerator2
    Skeletal Mesh=$System.Skeletal Mesh
    Socket Sampling Mode=::NewEnumerator5
   )
   &SubUVAnimation(
    "Sprite Renderer "=DataInterface() {propertyValues {SpriteRenderer=&self:Tetris_HandsFeet_0.NiagaraSpriteRendererProperties_0}}
    SubUV Animation Mode:ENiagara_SubUVLookupModeV2=::NewEnumerator2
    UseEndFrame=false
    UseStartFrame=false
   )
   &AddVelocity(
    Module.Use Falloff Distance=false
    Origin Offset:NiagaraPosition=$Particles.SkeletalMeshLocation.SampledPosition
    Override Seed=false
    Velocity Mode=::NewEnumerator1
    Velocity Speed=&FN_PickFloatFromSpawnGroupCutoff(
     Over Cutoff Float:NiagaraFloat=&RandomRangeFloat(Minimum=10, Maximum=45)
     Under Cutoff Float:NiagaraFloat=&RandomRangeFloat(Minimum=25, Maximum=75)
     SpawnGroup:NiagaraInt32=$Emitter.SpawnGroup
     Cutoff:NiagaraInt32=2
    )
   ) {module=AddVelocity001}
   &InitialMeshOrientation(Use Rotation Vector=false)
   &CameraOffset(Camera Offset Amount=&RandomRangeFloat(Minimum=-25, Maximum=-45))
  ]
  particleUpdateScript [
   &ParticleState()
   &ScaleMeshSize(Scale Factor=&VectorFromFloat(Value=&FloatFromCurve(
      FloatCurve=DataInterface() {propertyValues {Curve {
         defaultValue=0
         keys [
          {tangentMode=RCTM_User, time=0, value=1}
          {arriveTangent=-0.0405, interpMode=RCIM_Cubic, leaveTangent=-0.0405, tangentMode=RCTM_User, time=0.7009, value=1}
          {arriveTangent=-8.0089, interpMode=RCIM_Cubic, leaveTangent=-8.0089, tangentMode=RCTM_User, time=1, value=0}
         ]
        }}}
      Scale Curve=&Lerp_Float(A=1, B=0.7, Alpha=$User.ParaGlideLeanAlpha)
     )))
   &ScaleSpriteSize(Uniform Curve Sprite Scale=DataInterface() {propertyValues {Curve {
       defaultValue=0
       keys [
        {tangentMode=RCTM_User, time=0, value=1}
        {arriveTangent=-0.0405, interpMode=RCIM_Cubic, leaveTangent=-0.0405, tangentMode=RCTM_User, time=0.7009, value=1}
        {arriveTangent=-8.0089, interpMode=RCIM_Cubic, leaveTangent=-8.0089, tangentMode=RCTM_User, time=1, value=0}
       ]
      }}})
   &SpriteRotationRate(Rotation Rate=&RandomRangeFloat(Minimum=-90, Maximum=90))
   &MeshRotationForce(Rotation=&VectorFromFloat(Value=&RandomRangeFloat(Minimum=-90, Maximum=90)))
   &CurlNoiseForce(
    Bias Noise=false
    Noise Strength=&FloatFromCurve(
     FloatCurve=DataInterface() {propertyValues {Curve {keys [<at 0 0>,<at 0.2 1 RCIM_Cubic>]}}}
     Scale Curve=&RandomRangeFloat(Minimum=125, Maximum=300)
    )
    Pan Noise=false
   )
   &GravityForce(Gravity=&MakeVector(
     X=&MaskFloatBySpawnGroup(
      Masked Float=&Sine(Normalized Angle=$Emitter.Age, Period=2, Scale=&RandomRangeFloat(Minimum=650, Maximum=925), Bias=0)
      Passthrough Float=&MaskFloatBySpawnGroup(
       Masked Float=&Sine(Normalized Angle=$Emitter.Age, Period=-1.7, Scale=&RandomRangeFloat(Minimum=650, Maximum=925), Bias=0)
       Passthrough Float=&RandomRangeFloat(Minimum=-700, Maximum=700)
       Spawn Group Mask=1
      )
     )
     Y=&RandomRangeFloat(Minimum=-700, Maximum=700)
     Z=&FloatFromCurve(
      FloatCurve=DataInterface() {propertyValues {Curve {keys [<at 0 0 RCIM_Cubic>,<at 0.2 1 RCIM_Cubic>]}}}
      Scale Curve=&Lerp_Float(
       A=&RandomRangeFloat(Minimum=250, Maximum=1160)
       B=&RandomRangeFloat(Minimum=1160, Maximum=1400)
       Alpha=$User.ParaGlideLeanAlpha
      )
     )
    ))
   &Drag(
    Drag=&Lerp_Float(A=&RandomRangeFloat(Minimum=0.8, Maximum=6), B=0, Alpha=$User.ParaGlideLeanAlpha)
    Rotational Drag:NiagaraFloat=&RandomRangeFloat(Minimum=0.8, Maximum=2)
    Use Linear Drag=true
    Use Rotational Drag=true
   )
   &SolveForcesAndVelocity(Clamp Velocity=false, Limit Acceleration=false)
   &ScaleColor(
    Scale Alpha=&FloatFromCurve(
     FloatCurve=DataInterface() {propertyValues {Curve {
        defaultValue=1
        keys [
         {arriveTangent=9.0725, interpMode=RCIM_Cubic, leaveTangent=9.0725, tangentMode=RCTM_User, time=0, value=0}
         {arriveTangent=-0.0728, interpMode=RCIM_Cubic, leaveTangent=-0.0728, tangentMode=RCTM_User, time=0.1408, value=0.9847}
         {arriveTangent=-3.1322, interpMode=RCIM_Cubic, leaveTangent=-3.1322, tangentMode=RCTM_User, time=1, value=0}
        ]
       }}}
     Scale Curve=&CameraFade(
      Camera Query:NiagaraDataInterfaceCamera=DataInterface()
      Fade Offset:NiagaraFloat=50
      Fade Distance:NiagaraFloat=125
      Use Owner Position:NiagaraBool=true
     )
    )
    ScaleA=true
    ScaleRGB=true
   )
   /SetVariables(Particles.MeshColor:LinearColor=&MultiplyLinearColor(
     Linear Color=&SelectLinearColorFromArray(
      Array Sampling Mode=::NewEnumerator0
      Direct Array Index=?
      Color Selection Array=DataInterface() {propertyValues {ColorData [<rgba 0.3226 0.6488 1 1>, <rgba 1 0.9141 0.25 1>, <rgba 1 0.1771 0.7738 1>, <rgba 0.4684 1 0.2654 1>]}}
      Random Seed=0
     )
     Scale Factor=&Multiply_Vector4(
      A=&MakeVector4(
       X=1
       Y=1
       Z=1
       W=&FloatFromCurve(
        FloatCurve=DataInterface() {propertyValues {Curve {keys [<at 0 0 RCIM_Cubic>,<at 0.2 1 RCIM_Cubic>,<at 1 0 RCIM_Cubic>]}}}
        Scale Curve=&CameraFade(
         Camera Query:NiagaraDataInterfaceCamera=DataInterface()
         Fade Offset:NiagaraFloat=50
         Fade Distance:NiagaraFloat=125
         Use Owner Position:NiagaraBool=true
        )
       )
      )
      B=&MakeVector4_FromFloat(Float=&FN_Select_Float_from_Time_Of_Day(Morning=1.7, Day=1, Evening=1.7, Night=4))
     )
    )) {isSetParameters=true}
  ]
  rendererClasses [&NiagaraSpriteRendererProperties, &NiagaraMeshRendererProperties]
  renderers [
   {rendererClass=&NiagaraSpriteRendererProperties}
   {rendererClass=&NiagaraMeshRendererProperties, rendererIndex=1}
  ]
 }
}
summary {
 systemName=NS_Contrail_WiryPerk
 emitters [
  {emitterName=Ribbon_Feet, rendererClasses [&NiagaraRibbonRendererProperties]}
  {emitterName=Ribbon_Hands, rendererClasses [&NiagaraRibbonRendererProperties]}
  {
   emitterName=Tetris_HandsFeet
   rendererClasses [&NiagaraSpriteRendererProperties, &NiagaraMeshRendererProperties]
  }
  {emitterName=Pixel_Circles, rendererClasses [&NiagaraMeshRendererProperties]}
  {emitterName=Decal, rendererClasses [&NiagaraSpriteRendererProperties]}
  {emitterName=Pixels_HandsFeet, rendererClasses [&NiagaraMeshRendererProperties]}
 ]
 userVariables [
  User.bIsFrontEnd {type=NiagaraBool}
  User.bIsFrontEndPreview {type=NiagaraBool}
  User.ParaGlideLeanAlpha {type=NiagaraFloat, defaultValue=0, description=""}
  User.RibbonIDOffset {type=NiagaraInt32}
 ]
}
